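; Data asset for the Creative CoreGamer HUD that resolves team color sets for ownership and capture-progress displays in minigames.
; Ownership and progress lookups follow the minigame's UI color setting, either local-team relationship or per-team color, with progress also depending on capture state.
; State colors come from a map, including fallbacks for unowned, contested or invalid cases.
; Per-team sets derive darker shades by scaling CreativeTeamColors' bold color.
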(fn GetOwnershipTeamColor
  (params (InMinigame :type Fort%20Minigame%20Object%20Reference) (InLocalPlayerController :type Player%20Controller%20Object%20Reference) (Team :type Byte))
  (bind _returnvalue (Minigame|GetMinigameSettingsMachine InMinigame))
  (bind _returnvalue_2 (Minigame|GetMinigameTeamComponent InMinigame))
  (Utilities|IsValid InMinigame
    (:Is%20Valid
      (Utilities|IsValid _returnvalue
        (:Is%20Valid
          (if (== Team 0)
            (return (CallFunction|GetTeamColorbyState "NewEnumerator3"))
            (else
              (switch Utilities|FlowControl|Switch|SwitchonECreativeUITeamColors (Variables|/Script/FortniteGame.FortMinigameSettingsBuilding|GetUITeamColors _returnvalue)
                (:Relationship
                  (return (select (Default|IsLocalTeam (Variables|GetTeam) (Variables|GetInLocalPlayerController)) (CallFunction|GetTeamColorbyState "NewEnumerator1") (CallFunction|GetTeamColorbyState "NewEnumerator2"))))
                (:TeamColor
                  (Utilities|IsValid _returnvalue_2
                    (:Is%20Valid
                      (bind _outcolors (CallFunction|GetTeamColorbyIndex (Minigame|GameRules|GetTeamColorIndex _returnvalue_2 Team)))
                      (return _outcolors))
                    (:Is%20Not%20Valid
                      (return (CallFunction|GetTeamColorbyState "NewEnumerator0")))))))))
        (:Is%20Not%20Valid
          _)))
    (:Is%20Not%20Valid
      _))
  (return (CallFunction|GetTeamColorbyState "NewEnumerator0")))

(fn GetProgressTeamColor
  (params (CurrentMinigame :type Fort%20Minigame%20Object%20Reference) (InLocalPlayerController :type Player%20Controller%20Object%20Reference) (OwningTeam :type Byte) (CurrentStateTeam :type Byte) (CurrentState :type ECaptureState%20Enum :default "Neutral"))
  (bind _returnvalue (Minigame|GetMinigameSettingsMachine CurrentMinigame))
  (bind _currentminigame (Variables|GetCurrentMinigame))
  (bind _returnvalue_2 (Minigame|GetMinigameTeamComponent _currentminigame))
  (Utilities|IsValid CurrentMinigame
    (:Is%20Valid
      (Utilities|IsValid _returnvalue
        (:Is%20Valid
          (if (Utilities|Enum|Equal(Enum) (Variables|GetCurrentState) "Contested")
            ;[0.276 0.276 0.276] Zone is Contested
            (return (CallFunction|GetTeamColorbyState "NewEnumerator4"))
            (else
              ;[0.276 0.276 0.276] Zone is unowned. Colors Depends on Capture State
              (switch Utilities|FlowControl|Switch|SwitchonECreativeUITeamColors (Variables|/Script/FortniteGame.FortMinigameSettingsBuilding|GetUITeamColors _returnvalue)
                (:Relationship
                  ;;[0 0 0] Friendly Always one Color. Enemies always another
                  (switch Utilities|FlowControl|Switch|SwitchonECaptureState (Variables|GetCurrentState)
                    (:Neutral
                      _)
                    (:Capturing
                      (return (select (Default|IsLocalTeam (Variables|GetCurrentStateTeam) (Variables|GetInLocalPlayerController)) (CallFunction|GetTeamColorbyState "NewEnumerator1") (CallFunction|GetTeamColorbyState "NewEnumerator2"))))
                    (:Contested
                      (return (CallFunction|GetTeamColorbyState "NewEnumerator4")))
                    (:Disabled
                      _)
                    (:Decapturing
                      _)
                    (:Neutralizing
                      (return (select (Default|IsLocalTeam (Variables|GetCurrentStateTeam) (Variables|GetInLocalPlayerController)) (CallFunction|GetTeamColorbyState "NewEnumerator2") (CallFunction|GetTeamColorbyState "NewEnumerator1"))))
                    (:Deneutralizing
                      _)
                    (:Captured
                      _))
                  (return (CallFunction|GetTeamColorbyState "NewEnumerator3")))
                (:TeamColor
                  ;;[0 0 0] Per-Team Colors
                  (switch Utilities|FlowControl|Switch|SwitchonECaptureState (Variables|GetCurrentState)
                    (:Neutral
                      _)
                    (:Capturing
                      (return (CallFunction|GetOwnershipTeamColor (Variables|GetCurrentMinigame) (Variables|GetInLocalPlayerController) (Variables|GetCurrentStateTeam))))
                    (:Contested
                      (return (CallFunction|GetTeamColorbyState "NewEnumerator4")))
                    (:Disabled
                      _)
                    (:Decapturing
                      _)
                    (:Neutralizing
                      (Utilities|IsValid _returnvalue_2
                        (:Is%20Valid
                          (bind _outcolors (CallFunction|GetTeamColorbyIndex (Minigame|GameRules|GetTeamColorIndex _returnvalue_2 (Variables|GetOwningTeam))))
                          (return _outcolors))
                        (:Is%20Not%20Valid
                          (return (CallFunction|GetTeamColorbyState "NewEnumerator0")))))
                    (:Deneutralizing
                      _)
                    (:Captured
                      _))
                  (return (CallFunction|GetTeamColorbyState "NewEnumerator3")))))))
        (:Is%20Not%20Valid
          _)))
    (:Is%20Not%20Valid
      _))
  (return (CallFunction|GetTeamColorbyState "NewEnumerator0")))

(fn GetTeamColorByIndex
  (params (TeamColorIndex :type Integer))
  (bind (_teamlightcolors _teamselectorcolors _teamboldcolors) (Class|GetClassDefaults "/Game/Creative/Sets/MilitaryBase/Props/CreativeTeamColors.CreativeTeamColors_C"))
  (if (Utilities|Array|IsValidIndex _teamboldcolors TeamColorIndex)
    (return (Utilities|Struct|MakeSUICreativeCoreGamerTeamColors (Utilities|Array|Get(aref) _teamboldcolors TeamColorIndex) (* (Utilities|Array|Get(aref) _teamboldcolors TeamColorIndex) (Math|Conversions|ToLinearColor(Float) 0.9)) (* (Utilities|Array|Get(aref) _teamboldcolors TeamColorIndex) (Math|Conversions|ToLinearColor(Float) 0.75)) (* (Utilities|Array|Get(aref) _teamboldcolors TeamColorIndex) (Math|Conversions|ToLinearColor(Float) 0.5)) (* (Utilities|Array|Get(aref) _teamboldcolors TeamColorIndex) (Math|Conversions|ToLinearColor(Float) 0.25)) (* (Utilities|Array|Get(aref) _teamboldcolors TeamColorIndex) (Math|Conversions|ToLinearColor(Float) 0.15))))
    (else
      (return (Utilities|Struct|MakeSUICreativeCoreGamerTeamColors "(R=1.000000,G=1.000000,B=1.000000,A=0.000000)")))))

(fn GetTeamColorByState
  (params (InTeamState :type E_UI_Creative_CoreGamer_TeamStates%20Enum :default "NewEnumerator0"))
  (return (Utilities|Map|Find (Variables|Default|GetTeamStateColors) InTeamState)))

(fn IsLocalTeam
  (params (InTeamByte :type Byte) (LocalPlayerController :type Player%20Controller%20Object%20Reference))
  (return (== InTeamByte (Team|GetActorTeam LocalPlayerController))))
